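# Sound cue for the appearance of the front-end About modal screen.
# A single wave player feeds one More Info modal appear wave straight to the output at reduced volume.
# It plays in the non-critical in-game UI sound class, and a concurrency override allows one instance at a time, blocking new plays while one is active.
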SoundCue /Game/Sounds/Fort_UI_Sounds/FrontEnd/About_Modal/AboutModal_Screen_Appear_Cue
root=WavePlayer_0
nodes [{id=WavePlayer_0, properties {soundWaveAssetPtr=/Game/Sounds/Fort_UI_Sounds/FrontEnd/About_Modal/wav/UI_S24_MSE_MoreInfo_Modal_Appear}}]
properties {
 maxDistance=2097152
 soundClassObject=&Fort_UI_InGame_NonCritical
 volumeMultiplier=0.7
 concurrencyOverrides {maxCount=1, resolutionRule=PreventNew}
}
